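# Reward token widget, derived from FortShowdownDetailView, announcing that the player unlocked a new tournament.
# On Refresh Data, it fills three stacked text blocks: header, achievement and qualification message.
# The achievement text names the tournament, plus the round when visual rounds exist, falling back to a generic line if TitleLine1 is empty.
# The next-round qualification line collapses only when no visual rounds remain and the event ID is empty.
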WidgetBlueprint /FortniteCoreUI/UI/Events/Rewards/EventRewardToken
graphs [EventGraph {nodes [
   Event Refresh Data() {is=title K2Node_Event}
   > "SetText (Text)"(self=Get HeaderText() -> (HeaderText:Common Text Object Reference), InText="LOCTEXT(\"Unlocked a New Tournament!\")")
   n1 = Break Fort Tournament Display Info!(FortTournamentDisplayInfo:"Fort Tournament Display Info Structure (by ref)"=Get TournamentDisplayInfo() -> (TournamentDisplayInfo:Fort Tournament Display Info Structure)) -> (
    TitleLine1:Text
    TitleLine2:Text
    ScheduleInfo:Text
    FlavorDescription:Text
    DetailsDescription:Text
    ShortFormatTitle:Text
    LongFormatTitle:Text
    BackgroundTitle:Text
    PinScoreRequirement:int
    PinEarnedText:Text
    BaseColor:Linear Color Structure
    PrimaryColor:Linear Color Structure
    SecondaryColor:Linear Color Structure
    HighlightColor:Linear Color Structure
    TitleColor:Linear Color Structure
    ShadowColor:Linear Color Structure
    BackgroundLeftColor:Linear Color Structure
    BackgroundRightColor:Linear Color Structure
    BackgroundTextColor:Linear Color Structure
    PosterFadeColor:Linear Color Structure
    AlertText:Text
    SeriesPointLeaderboardName:Text
    AlertType:EFortTournamentAlertType Enum
    RoundDisplayNames:Array of Texts
   ) {is=title K2Node_BreakStruct}
   n2 = "integer > integer"(A=GetNumEventVisualRounds!(self:Fort Showdown Detail View Object Reference) -> (ReturnValue:int=0), B=0) {is=title K2Node_CallFunction}
   n3 = Format Text!(
    Format="LOCTEXT(\"You nailed it in {RoundDisplayName} of {TournamentNameLine1} {TournamentNameLine2}!\")"
    RoundDisplayName:Text=GetWindowRoundDisplayString!(self:Fort Showdown Detail View Object Reference) -> (ReturnValue:Text)
    TournamentNameLine1:Text=@n1
    TournamentNameLine2:Text=@n1.TitleLine2
   ) -> (Result)
   n4 = Format Text!(
    Format="LOCTEXT(\"You nailed it in {TournamentNameLine1} {TournamentNameLine2}!\")"
    TournamentNameLine1=@n1
    TournamentNameLine2=@n1.TitleLine2
   ) -> (Result)
   > "SetText (Text)"(
    self=Get AchievementText() -> (AchievementText:Common Text Object Reference)
    InText=Select(Option 0:Text=Select(Option 0=@n4, Option 1=@n3, Index=@n2), Option 1:Text="LOCTEXT(\"You nailed it!\")", Index=TextIsEmpty(InText=@n1)) -> (ReturnValue:Text)
   )
   > "SetText (Text)"(self=Get UnlockmentText() -> (UnlockmentText:Common Text Object Reference), InText="LOCTEXT(\"You\\'ve qualified for the next round!\")")
   n5 = IsEmpty(InString=GetEventId!(self:Fort Showdown Detail View Object Reference) -> (ReturnValue:String))
   n6 = NOT Boolean(A=HasRemainingVisualRounds!(self:Fort Showdown Detail View Object Reference) -> (ReturnValue:bool))
   > SetVisibility(
    self=Get UnlockmentText()
    InVisibility=Select(Option 0:ESlateVisibility Enum=SelfHitTestInvisible, Option 1:ESlateVisibility Enum=Collapsed, Index=AND Boolean(A=@n6, B=@n5)) -> (ReturnValue:ESlateVisibility Enum)
   )
  ]}]
variables [WasRefreshed {type=boolean}]
widgets {
 info {inheritedWidgetCount=6, parentClass=/Script/FortniteUI.FortShowdownDetailView, rootWidgetClass=&VerticalBox, widgetCount=10}
 widgets [
  VerticalBox_1 {widgetClassPath=&VerticalBox, widget=VerticalBox_1}
  HeaderText {
   widgetClassPath=&CommonTextBlock
   bIsVariable=true
   parent=VerticalBox_1
   widget=HeaderText
   properties {
    justification=Center
    text="Unlocked New Event!"
    wrapTextAt=850
    font {
     bForceMonospaced=false
     bMaterialIsStencil=false
     fontMaterial=null
     fontObject=null
     letterSpacing=0
     monospacedWidth=1
     size=24
     skewAmount=0
     typefaceFontName=None
     outlineSettings {
      bApplyOutlineToDropShadows=false
      bMiteredCorners=false
      bSeparateFillAlpha=false
      outlineMaterial=null
      outlineSize=0
      outlineColor=<rgba 0 0 0 1>
     }
    }
    shadowOffset=<0 0>
   }
   slot {class=VerticalBoxSlot}
  }
  AchievementText {
   widgetClassPath=&CommonTextBlock
   bIsVariable=true
   parent=VerticalBox_1
   widget=AchievementText
   properties {
    justification=Center
    text="You achieved something!"
    wrapTextAt=850
    wrappingPolicy=AllowPerCharacterWrapping
    colorAndOpacity {colorUseRule=UseColor_Specified, specifiedColor=<rgba 0.5776 0.5776 0.5776 1>}
    font {
     bForceMonospaced=false
     bMaterialIsStencil=false
     fontMaterial=null
     fontObject=&BurbankSmall
     letterSpacing=0
     monospacedWidth=1
     size=24
     skewAmount=0
     typefaceFontName=NotoSans-Regular
     outlineSettings {
      bApplyOutlineToDropShadows=false
      bMiteredCorners=false
      bSeparateFillAlpha=false
      outlineMaterial=null
      outlineSize=0
      outlineColor=<rgba 0 0 0 1>
     }
    }
   }
   slot {class=VerticalBoxSlot, properties {verticalAlignment=VAlign_Bottom, padding=<ltrb 0 20 0 0>}}
  }
  UnlockmentText {
   widgetClassPath=&CommonTextBlock
   bIsVariable=true
   parent=VerticalBox_1
   widget=UnlockmentText
   properties {
    justification=Center
    text=You unlocked something
    wrapTextAt=850
    wrappingPolicy=AllowPerCharacterWrapping
    colorAndOpacity {colorUseRule=UseColor_Specified, specifiedColor=<rgba 0.5776 0.5776 0.5776 1>}
    font {
     bForceMonospaced=false
     bMaterialIsStencil=false
     fontMaterial=null
     fontObject=&BurbankSmall
     letterSpacing=0
     monospacedWidth=1
     size=24
     skewAmount=0
     typefaceFontName=NotoSans-Regular
     outlineSettings {
      bApplyOutlineToDropShadows=false
      bMiteredCorners=false
      bSeparateFillAlpha=false
      outlineMaterial=null
      outlineSize=0
      outlineColor=<rgba 0 0 0 1>
     }
    }
   }
   slot {class=VerticalBoxSlot, properties {verticalAlignment=VAlign_Bottom, padding=<ltrb 0 20 0 0>}}
  }
  Text_TimeRemaining {widgetClassPath=&CommonTextBlock, bInherited=true, widget=null}
  Text_DebugId {widgetClassPath=&TextBlock, bInherited=true, widget=null, properties=~}
  Image_PosterFront {widgetClassPath=/Script/EpicCMSUIFramework.EpicCMSImage, bInherited=true, widget=null}
  Image_PosterBack {widgetClassPath=&EpicCMSImage, bInherited=true, widget=null}
  Image_LoadingScreen {widgetClassPath=&EpicCMSImage, bInherited=true, widget=null}
  Image_PlaylistTile {widgetClassPath=&EpicCMSImage, bInherited=true, widget=null}
 ]
}
