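# Gift-box message screen, built on FortMessageReceivedScreen, that announces an unlocked tournament round.
# When the gift box is set, its eventId and eventWindowId params feed ShowdownPin and EventRewardToken as TournamentId and EventId respectively.
# Construct switches to the default front-end camera and plays an Intro that pops in pin and token, then reveals Accept.
# Accept deactivates the screen, and deactivation runs NavigateForGift and broadcasts Message Closed.
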WidgetBlueprint /FortniteCoreUI/UI/ItemReceived/TournamentRoundUnlocked
animations [Intro {
  binding_count=3
  display_label=""
  length_seconds=1.3001
  object=&self:Intro
  sequence {bindings [
    Btn_Accept {
     id=#1
     possessed_class=WBP_UIKit_Button_Regular_C
     tracks [
      {
       type=MovieSceneFloatTrack
       display_name=Render Opacity
       sections [{
         type=MovieSceneFloatSection
         end_frame=26
         start_frame=0
         channels [None {type=MovieSceneScriptingFloatChannel, keys [[0,0],[11.001,0],[26.001,1]]}]
        }]
      }
      {
       type=MovieScene2DTransformTrack
       sections [{
         type=MovieScene2DTransformSection
         end_frame=18
         start_frame=9
         channels [
          Translation.X {type=MovieSceneScriptingFloatChannel}
          Translation.Y {type=MovieSceneScriptingFloatChannel, thinned_from=4, keys [[9,50],[18.0007,0],[18.001,0]]}
          Angle {type=MovieSceneScriptingFloatChannel}
          Scale.X {type=MovieSceneScriptingFloatChannel}
          Scale.Y {type=MovieSceneScriptingFloatChannel}
          Shear.X {type=MovieSceneScriptingFloatChannel}
          Shear.Y {type=MovieSceneScriptingFloatChannel}
         ]
        }]
      }
     ]
    }
    ShowdownPin {
     id=#2
     possessed_class=ShowdownPin_C
     tracks [{
       type=MovieScene2DTransformTrack
       sections [{
         type=MovieScene2DTransformSection
         end_frame=11
         start_frame=0
         channels [
          Translation.X {type=MovieSceneScriptingFloatChannel}
          Translation.Y {type=MovieSceneScriptingFloatChannel}
          Angle {type=MovieSceneScriptingFloatChannel}
          Scale.X {type=MovieSceneScriptingFloatChannel, keys [[0,0],[3.999,1.3],[7,1]]}
          Scale.Y {type=MovieSceneScriptingFloatChannel, keys [[0,0],[3.999,1.3],[7.0003,1]]}
          Shear.X {type=MovieSceneScriptingFloatChannel}
          Shear.Y {type=MovieSceneScriptingFloatChannel}
         ]
        }]
      }]
    }
    EventRewardToken {
     id=#3
     possessed_class=EventRewardToken_C
     tracks [{
       type=MovieScene2DTransformTrack
       sections [{
         type=MovieScene2DTransformSection
         end_frame=16
         start_frame=0
         channels [
          Translation.X {type=MovieSceneScriptingFloatChannel}
          Translation.Y {type=MovieSceneScriptingFloatChannel}
          Angle {type=MovieSceneScriptingFloatChannel}
          Scale.X {type=MovieSceneScriptingFloatChannel, keys [[6.001,0],[11.001,1.2],[13.001,1]]}
          Scale.Y {type=MovieSceneScriptingFloatChannel, keys [[6.001,0],[11.001,1.2],[13.001,1]]}
          Shear.X {type=MovieSceneScriptingFloatChannel}
          Shear.Y {type=MovieSceneScriptingFloatChannel}
         ]
        }]
      }]
    }
   ]}
 }]
graphs [EventGraph {nodes [
   "On Clicked (Btn_Accept)"!() -> (OutputDelegate, then, Button:Common Button Base Object Reference) {is=title K2Node_ComponentBoundEvent}
   > DeactivateWidget()
   Event OnGiftBoxSet() {is=title K2Node_Event}
   n1 = Get GiftBoxItem() -> (GiftBoxItem:Fort Gift Box Item Object Reference)
   n2 = TryGetParamString!(self:Fort Gift Box Item Object Reference=@n1, ParamName:String=eventWindowId) -> (OutValue:String, ReturnValue:bool)
   n3 = TryGetParamString(self=@n1, ParamName=eventId)
   > SetTournament!(
    execute
    self:Fort Showdown Detail View Object Reference=Get EventRewardToken() -> (EventRewardToken:Event Reward Token Object Reference)
    TournamentId:String=@n3
    EventId:String=@n2
    bForceRefreshUI:bool=true
   ) -> (then)
   > SetTournament(self=Get ShowdownPin() -> (ShowdownPin:Showdown Pin Object Reference), TournamentId=@n3, EventId=@n2, bForceRefreshUI=true)
   Event Construct()
   > SetFrontEndCamera!(
    execute
    self:Fort Front End Context Object Reference=Get FortFrontEndContext!() -> (ReturnValue:Fort Front End Context Object Reference) {is=title K2Node_GetBlueprintContextEx}
    NewState:EFrontEndCamera Enum=FrontendDefault
    TransitionParams:View Target Transition Params Structure=(BlendTime=0,BlendFunction=VTBlend_Cubic,BlendExp=2,bLockOutgoing=False)
   ) -> (then)
   > PlayAnimation(InAnimation=Get Intro())
   Event On Deactivated()
   > NavigateForGift!(execute, Item:Fort Gift Box Item Object Reference=Get GiftBoxItem()) -> (then)
   > Call On Message Closed!(execute:Exec, self:Tournament Round Unlocked Object Reference) -> (then:Exec) {is=title K2Node_CallDelegate}
   Event Tick()
  ]}]
variables []
widgets {
 info {inheritedWidgetCount=1, parentClass=/Script/FortniteUI.FortMessageReceivedScreen, rootWidgetClass=&Overlay, widgetCount=8}
 widgets [
  Overlay_2 {widgetClassPath=&Overlay, widget=Overlay_2}
  WBP_UIKit_Backplate {
   widgetClassPath=&WBP_UIKit_Backplate
   bIsVariable=true
   parent=Overlay_2
   widget=WBP_UIKit_Backplate
   properties {brightness=Dark}
   slot {class=OverlaySlot, properties {horizontalAlignment=HAlign_Fill, verticalAlignment=VAlign_Fill}}
  }
  MainSafeZone {
   widgetClassPath=&SafeZone
   parent=Overlay_2
   widget=MainSafeZone
   slot {class=OverlaySlot, properties {horizontalAlignment=HAlign_Fill, verticalAlignment=VAlign_Fill}}
  }
  MainContainer {
   widgetClassPath=&VerticalBox
   parent=MainSafeZone
   widget=MainContainer
   slot {class=SafeZoneSlot, properties {hAlign=HAlign_Center, vAlign=VAlign_Center}}
  }
  ShowdownPin {
   widgetClassPath=/FortniteCoreUI/Frontend/Showdown/ShowdownPin
   bIsVariable=true
   parent=MainContainer
   widget=ShowdownPin
   properties {dEVPinScore=0}
   slot {class=VerticalBoxSlot}
  }
  EventRewardToken {
   widgetClassPath=/FortniteCoreUI/UI/Events/Rewards/EventRewardToken
   bIsVariable=true
   parent=MainContainer
   widget=EventRewardToken
   slot {class=VerticalBoxSlot, properties {horizontalAlignment=HAlign_Center, verticalAlignment=VAlign_Center, padding=<ltrb 0 4 0 0>}}
  }
  Btn_Accept {
   widgetClassPath=&WBP_UIKit_Button_Regular
   bIsVariable=true
   parent=MainContainer
   widget=Btn_Accept
   properties {text=Accept}
   slot {class=VerticalBoxSlot, properties {horizontalAlignment=HAlign_Center, verticalAlignment=VAlign_Bottom, padding=<ltrb 0 50 0 0>}}
  }
  GiftBoxImage {widgetClassPath=/Script/CommonUI.CommonLazyImage, bInherited=true, widget=null}
 ]
}
