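# UI sound cue for an item being shown in the front-end gifting flow.
# A Mixer layers the gifting item-shown wave with a rewards-appear wave, the latter run through a Modulator that fixes it at higher pitch and lower volume.
# The cue uses the Fort_UI sound class, applies its own pitch and volume multipliers, and overrides concurrency to a small limit that stops the oldest instance.
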SoundCue /Game/Sounds/Fort_UI_Sounds/FrontEnd/RewardsAndGifts/Gifting_ItemShown_Cue
root=Mixer_0
nodes [
 {id=Mixer_0, children [WavePlayer_1,Modulator_0], properties {inputVolume [1,1]}}
 {id=Modulator_0, children [WavePlayer_0], properties {pitchMax=2, pitchMin=2, volumeMax=0.4, volumeMin=0.4}}
 {id=WavePlayer_0, properties {soundWaveAssetPtr=/Game/Sounds/Fort_UI_Sounds/FrontEnd/New_FrontEnd/UI_Rewards_Appear_Center_01}}
 {id=WavePlayer_1, properties {soundWaveAssetPtr=/Game/Sounds/Fort_UI_Sounds/FrontEnd/New_FrontEnd/gifting_itemshown_01}}
]
properties {
 bOverrideConcurrency=true
 maxDistance=2097152
 pitchMultiplier=2
 soundClassObject=&Fort_UI
 volumeMultiplier=0.4
 concurrencyOverrides {maxCount=2, resolutionRule=StopOldest}
}
